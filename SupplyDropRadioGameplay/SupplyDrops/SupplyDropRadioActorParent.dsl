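; Base actor for radio-called supply drops that fall under a balloon and deliver goods, such as crates or vehicles.
; On the server it loads flare data, teleports to a clear fall height, spawns a landing indicator and adds relevancy components to attached actors.
; Replicated balloon state drives gravity and cues; replicated ground info controls tick and landing, resuming falls when ground is destroyed.
; Children override hooks; Blueprint sweep-movement functions are deprecated.
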(event Building|EventOnBuildingActorInitialized
  (params (InitializationReason :type EFortBuildingInitializationReason%20Enum :default "StaticallyPlaced") (BuildingPersistentState :type EFortBuildingPersistentState%20Enum :default "Default"))
  ;[0.01848 0.1406 0.1294] BuildingActorInitialized: When the building actor is spawned, execute a spawn gameplay cue on the client.
  (|CallParentFunction InitializationReason BuildingPersistentState)
  (CallFunction|ApplyAttachComponenttoAttachedActors) ; By this point, all actors are expected to be attached.
  (Ability|ExecuteGameplayCueReplicated (Variables|Default|GetSpawnGameplayCueTag) (Class|AthenaFunctionLibrary|MakeSimpleFortEffectContext self self)))

(event EventBeginPlay
  (bind _balloonactor (CallFunction|GetBalloonActor))
  (bind _returnvalue (Ability|GetAbilitySystemComponent _balloonactor))
  ;[0 0.2135 0.02302] BeginPlay: On the server, move the supply drop to the desired fall height.
  ; Also, setup the balloon destroy quest for the balloon child actor.
  (|CallParentFunction)
  (CallFunction|InitData)
  (Networking|SwitchHasAuthority
    (:Authority
      (Fortnite|FortShippingLog self (Utilities|String|Append "BeginPlay: Spawned: Instigator" (Utilities|GetDisplayName (Game|GetInstigator))))
      (SupplyDropRadio|SetGravityMultiplier (Variables|Default|GetGravityMultiplier_PreOpen))
      (CallFunction|TeleporttoDesiredHeight)
      ;;[0 0.2604 0.2469] Spawn indicator on the ground where the supply drop is going to land.
      (bind _returnvalue_2 (Game|SpawnActorfromClass (Variables|Default|GetSupplyDropRadioIndicator%20Class) (Math|Conversions|ToTransform(Vector) (Variables|SupplyDropRadio|GetOriginalServerSpawnLocation)) "AlwaysSpawn" "OverrideRootScale" self (Variables|Default|GetFallHeight) (CallFunction|GetFlareDuration) 0))
      (Variables|Default|SetIndicator _returnvalue_2)
      (Variables|SupplyDropRadio|SetServerLocation (Transformation|GetActorLocation))
      (SupplyDropRadio|SetBalloonState "Closed")
      (Utilities|IsValid _balloonactor
        (:Is%20Valid
          (Class|SupplyDropRadioBalloonParent|SetSupplyDropRadioActorParent _balloonactor self :!class /SupplyDropRadioGameplay/SupplyDrops/SupplyDropRadioBalloonParent.SupplyDropRadioBalloonParent_C)
          (Default|BindEventtoOnBalloonDestroyed _balloonactor (EventDispatchers|CreateEvent :!event HandleBalloonDestroyed))
          (Utilities|IsValid _returnvalue
            (:Is%20Valid
              (GameplayEffects|ApplyGameplayEffectToSelf :self _returnvalue :GameplayEffectClass (Variables|Default|GetSupplyDropFallingQuestGE) :Level 0.0))))
        (:Is%20Not%20Valid
          ;;[1 0 0] Warning: Balloon actor is null.
          (Fortnite|FortShippingLog self "BeginPlay: BalloonActor is null." true))))
    (:Remote
      (Variables|SupplyDropRadio|SetServerLocation (Transformation|GetActorLocation))
      (Collision|SetCollisionEnabled (Variables|Default|GetWaterCollision) "NoCollision") ; Server will tell the client If it is in water.\nDo not worry about generating overlaps on the client.
    )))

(event Custom|NotifySupplyDropComplete
  ;[0.05005 0.224 0.2039] NotifySupplyDropComplete: Called when the supply drop actor has delivered its goods.
  ; For crates, this is when the crate is opened.
  ; For vehicles, this is when the vehicle is detached from the balloon.
  (Networking|SwitchHasAuthority
    (:Authority
      (Fortnite|FortShippingLog self "NotifySupplyDropComplete: Invoked.")
      (Utilities|Time|SetTimerbyEvent (EventDispatchers|CreateEvent :!event DelayedSilentDie) 2.0)
      (Variables|Default|SetbIsDying true)
      (bind _indicator (Variables|Default|GetIndicator)
        (:then
          (Class|SupplyDropRadioIndicator|SetOwnerLanded _indicator :!class /SupplyDropRadioGameplay/SupplyDrops/SupplyDropRadioIndicator.SupplyDropRadioIndicator_C))
        (:else)))))

(event OnComponentBeginOverlap(WaterCollision)
  (params (OverlappedComponent :type Primitive%20Component%20Object%20Reference) (OtherActor :type Actor%20Object%20Reference) (OtherComp :type Primitive%20Component%20Object%20Reference) (OtherBodyIndex :type Integer) (bFromSweep :type Boolean) (SweepResult :type Hit%20Result%20Structure%20%28by%20ref%29))
  ;[0.08597 0.1289 0.1771] OnComponentBeginOverlap(WaterCollision): Invoked on the server whenever the WaterCollisionComponent touches the water.
  ; This is required since the TickMovement() function does NOT detect WaterBodies - the WorldCollision component is of object type Dynamic which WaterBodies ignore.
  ; Therefore, collision with water is detected through an overlap using a separate component. For perf reasons, this component ignores overlap with everything except water.
  (Networking|SwitchHasAuthority
    (:Authority
      (CallFunction|HandleWaterCollisionOverlap OtherActor) ; Separate event so children can override it.
    )))

(event Custom|HandleWaterCollisionOverlap
  (params (OtherActor :type Actor%20Object%20Reference))
  ;[0.08597 0.1289 0.1771] OnComponentBeginOverlap(WaterCollision): Invoked on the server whenever the WaterCollisionComponent touches the water.
  ; This is required since the TickMovement() function does NOT detect WaterBodies - the WorldCollision component is of object type Dynamic which WaterBodies ignore.
  ; Therefore, collision with water is detected through an overlap using a separate component. For perf reasons, this component ignores overlap with everything except water.
  (Networking|SwitchHasAuthority
    (:Authority
      (bind _asfort_water_body_actor (Utilities|Casting|CastToFortWaterBodyActor OtherActor))
      (Variables|SupplyDropRadio|SetWaterHitResult (Collision|MakeHitResult true false 0.0 0.0 (Transformation|GetActorLocation) "0, 0, 0" "0,0,1" "0,0,1" 0 _asfort_water_body_actor))
      (Collision|SetCollisionEnabled (Variables|Default|GetWaterCollision) "NoCollision")
      (Fortnite|FortShippingLog self (Utilities|String|Append "HandlewaterCollisionOverlap: Overlapped water body: " (Utilities|GetDisplayName _asfort_water_body_actor))))))

(event EventOnDeathServer
  (params (Damage :type Float%20%28single-precision%29) (DamageTags :type Gameplay%20Tag%20Container%20Structure%20%28by%20ref%29) (Momentum :type Vector :default "0, 0, 0") (HitInfo :type Hit%20Result%20Structure%20%28by%20ref%29) (InstigatedBy :type Controller%20Object%20Reference) (DamageCauser :type Actor%20Object%20Reference) (EffectContext :type Gameplay%20Effect%20Context%20Handle%20Structure))
  ;[0.01197 0.5975 0.6354] OnDeathServer: Reset the net cull distance for all the attached actors.
  (|CallParentFunction Damage DamageTags Momentum HitInfo InstigatedBy DamageCauser EffectContext)
  (Networking|SwitchHasAuthority
    (:Authority
      (for _array_element (Variables|Default|GetAttachActorComponents)
        (Utilities|IsValid _array_element
          (:Is%20Valid
            (Class|BPSupplyDropRadioAttachActorComponent|ResetNetCullDistanceSquared _array_element)
            (Fortnite|FortShippingLog self (Utilities|String|Append "OnDeathServer: Resetting relevancy range for actor : " (Utilities|GetDisplayName (Components|GetOwner _array_element))))))))))

(event EventEndPlay
  (params (EndPlayReason :type EEndPlayReason%20Enum :default "Destroyed"))
  (bind _balloonactor (CallFunction|GetBalloonActor))
  (bind (_groundinfo_bvalid_8_6eea3ba64f4126c2fde592aedf82163b _groundinfo_groundactor_2_3cc9853b48ea31722d3aeba61f41499c) (Utilities|Struct|BreakSupplyDropRadioGroundInfoStruct (Variables|GetGroundInfo_bValid_8_6EEA3BA64F4126C2FDE592AEDF82163B)))
  ;[0.1615 0.09764 0.01592] EndPlay: Unbind delegates.
  (|CallParentFunction EndPlayReason)
  (Utilities|FlowControl|Sequence
    (:then_0
      (Utilities|IsValid _groundinfo_groundactor_2_3cc9853b48ea31722d3aeba61f41499c
        (:Is%20Valid
          (bind _asbuilding_actor (Utilities|Casting|CastToBuildingActor _groundinfo_groundactor_2_3cc9853b48ea31722d3aeba61f41499c)
            (:then
              (EventDispatchers|UnbindEventfromOnDied _asbuilding_actor (EventDispatchers|CreateEvent :!event HandleGroundBuildingActorDied)))
            (:CastFailed
              (Game|UnbindEventfromOnDestroyed _groundinfo_groundactor_2_3cc9853b48ea31722d3aeba61f41499c (EventDispatchers|CreateEvent :!event HandleGroundActorDestroyed)))))))
    (:then_1
      (if (Utilities|Enum|Equal(Enum) EndPlayReason "Destroyed")
        (Fortnite|FortShippingLog self "EndPlay: Invoked: Reason: Destroyed")))
    (:then_2
      (Networking|SwitchHasAuthority
        (:Authority
          (Utilities|IsValid _balloonactor
            (:Is%20Valid
              (Default|UnbindEventfromOnBalloonDestroyed _balloonactor (EventDispatchers|CreateEvent :!event HandleBalloonDestroyed)))))))))

(event SupplyDropRadio|EventOpenSupplyDrop
  ;[0.01904 0.2552 0.1325] OpenSupplyDrop: Asks this actor to deliver/open whatever form of supply it is carrying.
  (Networking|SwitchHasAuthority
    (:Authority
      (Fortnite|FortShippingLog self "OpenSupplyDrop: Invoked.")
      (CallFunction|NotifySupplyDropComplete))))

(event SupplyDropRadio|EventOnRep_BalloonState
  ;[0.1899 0.06527 0.2083] OnRep_BalloonState: Broadcast that the balloon's state has changed.
  ; In addition, ExecuteGameplayCue if the balloon has opened.
  (Default|CallOnBalloonStateChanged (SupplyDropRadio|GetBalloonState))
  (Networking|SwitchHasAuthority
    (:Authority
      (if (Utilities|Enum|Equal(Enum) (SupplyDropRadio|GetBalloonState) "Opened")
        (Ability|ExecuteGameplayCueReplicated (Variables|Default|GetOpenGameplayCueTag) (Class|AthenaFunctionLibrary|MakeSimpleFortEffectContext self self))
        (Components|Activation|Deactivate (Variables|Default|GetEntryFX))
        (SupplyDropRadio|SetGravityMultiplier (Variables|Default|GetGravityMultiplier_PostOpen))))))

(event SupplyDropRadio|EventOnRep_DistanceFromOriginalSpawnLocForBalloonOpen
  ;[0.04277 0.165 0.1719] OnRep_DistanceFromOriginalSpawnLocForBalloonOpen: Once the distance is received on the client, setup the entry FX duration.
  (Networking|CanExecuteCosmeticEvents
    (:True
      (CallFunction|SetupEntryFXTime))))

(fn ApplyAttachComponentToAttachedActors
  ;[0.03553 0.1408 0.1719] ApplyAttachComponentToAttachedActors: Add the component that is going to manage the attached actors.
  (Networking|SwitchHasAuthority
    (:Authority
      (for _array_element (Actor|GetAttachedActors true true)
        (Utilities|IsValid _array_element
          (:Is%20Valid
            (bind _badd (CallFunction|ShouldAddAttachComponentforActor _array_element))
            (if _badd
              (Utilities|IsValid (Actor|GetComponentbyClass _array_element "/SupplyDropRadioGameplay/Gameplay/BP_SupplyDropRadioAttachActorComponent.BP_SupplyDropRadioAttachActorComponent_C")
                (:Is%20Valid
                  ;;[1 0 0] Warning: Component already exists.
                  (Fortnite|FortShippingLog self (Utilities|String|Append "ApplyAttachComponentToAttachedActors: Failed to spawn component for actor as it already exists on the actor: " (Utilities|GetDisplayName _array_element)) true))
                (:Is%20Not%20Valid
                  (bind _returnvalue (Game|AddComponentbyClass _array_element "/SupplyDropRadioGameplay/Gameplay/BP_SupplyDropRadioAttachActorComponent.BP_SupplyDropRadioAttachActorComponent_C"))
                  (Utilities|IsValid _returnvalue
                    (:Is%20Valid
                      (Utilities|Array|Add (Variables|Default|GetAttachActorComponents) _returnvalue)
                      (Class|BPSupplyDropRadioAttachActorComponent|SetNetCullDistanceSquared _returnvalue (CallFunction|GetDesiredNetCullDistanceSquaredforAttachedActor _array_element))
                      (Fortnite|FortShippingLog self (Utilities|String|Append "ApplyAttachComponentToAttachedActors: Setup component for attached actor: " (Utilities|GetDisplayName _array_element))))
                    (:Is%20Not%20Valid
                      ;;[1 0 0] Warning: Failed to spawn component for a specific actor.
                      (Fortnite|FortShippingLog self (Utilities|String|Append "ApplyAttachComponentToAttachedActors: Failed to spawn component for actor : " (Utilities|GetDisplayName _array_element)) true)))))
              (else
                (Fortnite|FortShippingLog self (Utilities|String|Append "ApplyAttachComponentToAttachedActors: Not adding component for attached actor : " (Utilities|GetDisplayName _array_element)))))))))))

(fn AutoMarkOnSpawn
  (params (ActorToMark :type Actor%20Object%20Reference))
  ;[0.1839 0.08405 0.1927] AutoMarkOnSpawn: The instigator player auto marks the spawned supply drop for their team.
  ; Call this from the child actor whenever the actor to mark is ready.
  (Networking|SwitchHasAuthority
    (:Authority
      (Utilities|IsValid ActorToMark
        (:Is%20Valid
          (if (ScalableFloat|GetValueasBool (Variables|Hotfix|GetHotfix_AutoMarkOnSpawn))
            (bind _instigator (Variables|Actor|GetInstigator)
              (:then
                (Utilities|IsValid (Fortnite|GetFortPlayerControllerfromActor _instigator)
                  (:Is%20Valid
                    (Utilities|IsValid (Class|FortPlayerController|GetMarkerComponent (Fortnite|GetFortPlayerControllerfromActor _instigator))
                      (:Is%20Valid
                        (MarkerComponent|MarkActoronServer (Class|FortPlayerController|GetMarkerComponent (Fortnite|GetFortPlayerControllerfromActor _instigator)) ActorToMark true false))))))
              (:else))))))))

(fn DelayedSilentDie
  ;[0.03475 0.2912 0.3385] DelayedSilentDie: Called a short duration after the supply drop complete notify is fired. This allows for any client side effects to play out.
  (Networking|SwitchHasAuthority
    (:Authority
      (Gameplay|SilentDie false))))

(fn DestroyBalloonActor
  (params (bSilent :type Boolean :default true))
  (bind _balloonactor (CallFunction|GetBalloonActor))
  ;[0.1807 0.2083 0.0346] DestroyBalloonActor: Called to destroy the balloon child actor.
  (Networking|SwitchHasAuthority
    (:Authority
      (Utilities|IsValid _balloonactor
        (:Is%20Valid
          (Fortnite|FortShippingLog self (Utilities|String|Append "DestroyBalloonActor: Invoked: Silent ? " (Utilities|String|ToString(Boolean) bSilent)))
          (if bSilent
            (Gameplay|SilentDie _balloonactor false)
            (else
              (Gameplay|Die :self _balloonactor))))))))

(fn GetAttachedActorToMark
  ;[0.04189 0.151 0.1237] Overridden in children.
  (return 0))

(fn GetBalloonActor
  ;[0.02071 0.175 0.1823] Returns the balloon child actor.
  (bind _balloonactorchild (Variables|Default|GetBalloonActorChild)
    (:then
      (bind _assupply_drop_radio_balloon_parent (Utilities|Casting|CastToSupplyDropRadioBalloonParent (Variables|/Script/Engine.ChildActorComponent|GetChildActor _balloonactorchild))
        (:then
          (return _assupply_drop_radio_balloon_parent))
        (:CastFailed
          _)))
    (:else
      _))
  (return 0))

(fn GetClosedStateFallDuration
  (bind (_returnvalue_x _returnvalue_y _returnvalue_z) (Math|Vector|BreakVector (Transformation|GetActorLocation)))
  ;[0.04068 0.1302 0.1289] Returns for how long the supply drop is going to fall in its closed state.
  (bind _output_get (LocalVariables|SetFallDistToOpen (- _returnvalue_z (Variables|SupplyDropRadio|GetDistanceFromOriginalSpawnLocForBalloonOpen))))
  (if (> _output_get 0)
    (return (Math|Float|Sqrt (/ (* 2.0 _output_get) (SupplyDropRadio|GetGravityZ))))
    (else
      (return 0.0))))

(fn GetDesiredNetCullDistanceSquaredForAttachedActor
  (params (InActor :type Actor%20Object%20Reference))
  (return (Variables|Replication|GetNetCullDistanceSquared)))

(fn GetFlareDuration
  (bind (_originalserverspawnlocation_x _originalserverspawnlocation_y _originalserverspawnlocation_z) (Math|Vector|BreakVector (Variables|GetOriginalServerSpawnLocation_X)))
  ;[0.2396 0.08716 0.188] Get the duration for which the supply drop flare should be active.
  (Networking|SwitchHasAuthority
    (:Authority
      (bind _value (CallFunction|GetClosedStateFallDuration))
      (return (* (+ _value (Math|Float|SafeDivide (Math|Float|Max(Float) (- (Variables|SupplyDropRadio|GetDistanceFromOriginalSpawnLocForBalloonOpen) _originalserverspawnlocation_z)) (Variables|SupplyDropRadio|GetBalloonOpenedFallSpeed))) (Variables|Default|GetFlareDurationMultiplier))))))

(fn GetGroundInfo
  ;[0.04026 0.1566 0.1615] Returns the ground info.
  (return (Variables|Default|GetGroundInfo)))

(fn GetRootCollisionComponent
  ;[0.04561 0.08476 0.151] Returns the blueprint defined root component for the native side.
  (return (Variables|Default|GetWorldCollision)))

(fn HandleBalloonDestroyed
  (bind _balloonactor (CallFunction|GetBalloonActor))
  ;[0.002711 0.1594 0.1771] HandleBalloonDestroyed: Delegate callback for when the balloon is destroyed.
  (Networking|SwitchHasAuthority
    (:Authority
      (SupplyDropRadio|SetBalloonState "Destroyed")
      (Utilities|IsValid _balloonactor
        (:Is%20Valid
          (Default|UnbindEventfromOnBalloonDestroyed _balloonactor (EventDispatchers|CreateEvent :!event HandleBalloonDestroyed)))))))

(fn HandleGroundActorDestroyed
  (params (DestroyedActor :type Actor%20Object%20Reference))
  ;[0.06136 0.2656 0.2365] HandleGroundActorDestroyed: Called whenever the ground actor is destroyed.
  (Networking|SwitchHasAuthority
    (:Authority
      (Utilities|IsValid DestroyedActor
        (:Is%20Valid
          (Game|UnbindEventfromOnDestroyed DestroyedActor (EventDispatchers|CreateEvent :!event HandleGroundActorDestroyed))
          (Variables|SetGroundInfo_bValid_8_6EEA3BA64F4126C2FDE592AEDF82163B (struct :bValid_8_6EEA3BA64F4126C2FDE592AEDF82163B false :GroundActor_2_3CC9853B48EA31722D3AEBA61F41499C 0 :GroundLocation_5_9924359A435909587F0606803CB2EF23 "0.000000,0.000000,0.000000"))
          (Collision|IgnoreActorwhenMoving (Variables|Default|GetWorldCollision) DestroyedActor true) ; Ignore the actor when falling since we consider it destroyed.
          (Fortnite|FortShippingLog self (Utilities|String|Append "HandleGroundActorDestroyed: Invoked: " (Utilities|GetDisplayName DestroyedActor))))))))

(fn HandleGroundBuildingActorDied
  (params (DamagedActor :type Actor%20Object%20Reference) (Damage :type Float%20%28single-precision%29) (InstigatedBy :type Controller%20Object%20Reference) (DamageCauser :type Actor%20Object%20Reference) (HitLocation :type Vector) (FHitComponent :type Primitive%20Component%20Object%20Reference) (BoneName :type Name) (Momentum :type Vector))
  ;[0.04391 0.1615 0.1238] HandleGroundBuildingActorDied: Called when the building actor ground dies.
  (Networking|SwitchHasAuthority
    (:Authority
      (bind _asbuilding_actor (Utilities|Casting|CastToBuildingActor DamagedActor))
      (EventDispatchers|UnbindEventfromOnDied _asbuilding_actor (EventDispatchers|CreateEvent :!event HandleGroundBuildingActorDied))
      (CallFunction|HandleGroundActorDestroyed DamagedActor))))

(fn HandleLandEvent
  (bind _returnvalue (Ability|GetAbilitySystemComponent (CallFunction|GetBalloonActor)))
  (bind _balloonactor (CallFunction|GetBalloonActor))
  ;[0.01633 0.09403 0.09896] HandleLandEvent: Remove the quest GE from the balloon actor since it is no longer falling.
  ; Called on both server and client.
  (Networking|SwitchHasAuthority
    (:Authority
      (Fortnite|FortShippingLog self "HandleLandEvent: Invoked.")
      (Variables|SupplyDropRadio|SetVelocity "0, 0, 0")
      (Utilities|IsValid _balloonactor
        (:Is%20Valid
          (Utilities|IsValid _returnvalue
            (:Is%20Valid
              (GameplayEffects|RemoveActiveGameplayEffectbySourceEffect _returnvalue (Variables|Default|GetSupplyDropFallingQuestGE) 0 -1)
              (Fortnite|FortShippingLog self "HandleLandEvent: SupplyDropFallingQuestGE removed."))))))))

(fn HandleMovementBlocked_Deprecated
  (params (HitResult :type Hit%20Result%20Structure%20%28by%20ref%29))
  (bind (_bblockinghit _binitialoverlap _time _distance _location _impactpoint _normal _impactnormal _physmat _hitactor) (Collision|BreakHitResult HitResult))
  (bind (_asfort_projectile_base _bsuccess) (Utilities|Casting|CastToFortProjectileBase _hitactor))
  (bind (_asfort_remote_controlled_pawn_athena _bsuccess_2) (Utilities|Casting|CastToFortRemoteControlledPawnAthena _hitactor))
  (bind (_aspawn _bsuccess_3) (Utilities|Casting|CastToPawn _hitactor))
  ;[0.03297 0.165 0.1667] HandleMovementBlocked: Assess the movement sweep hit result and return if the actor is blocked.
  ; Runs on both server and client.
  (Utilities|FlowControl|Sequence
    (:then_0
      ;;[0.03593 0.1264 0.1823] Ignore any actor that is ultimately attached to the transport actor.
      (Utilities|IsValid _hitactor
        (:Is%20Valid
          (bind _returnvalue (SupplyDropRadio|GetHighestAttachParent _hitactor))
          (if (== _returnvalue self)
            (Collision|IgnoreActorwhenMoving (Variables|Default|GetWorldCollision) _hitactor true)
            (Fortnite|FortShippingLog self (Utilities|String|Append "HandleMovementBlocked: IgnoringActorWhenMoving: " (Utilities|GetDisplayName _hitactor)))
            (else
              ;;[0.03409 0.151 0.1299] Supply drops falling on FortMountedTurrets are automatically opened.
              ;; Default supply drop behavior from AthenaSupplyDrop actor.
              (bind _asfort_mounted_turret (Utilities|Casting|CastToFortMountedTurret _hitactor)
                (:then
                  (Networking|SwitchHasAuthority
                    (:Authority
                      (Fortnite|FortShippingLog self (Utilities|String|Append "HandleMovementBlocked: Hit Turret Actor, opening supply drop: " (Utilities|GetDisplayName _asfort_mounted_turret)))
                      (SupplyDropRadio|OpenSupplyDrop))))
                (:CastFailed
                  (if (Math|Boolean|ORBoolean _bsuccess _bsuccess_2 _bsuccess_3)
                    (LocalVariables|SetbRetBlock false)
                    (else
                      (bind _asbuilding_actor (Utilities|Casting|CastToBuildingActor _hitactor)
                        (:then
                          (if (or (Utilities|Enum|Equal(Enum) (Variables|/Script/Engine.SceneComponent|GetMobility (Variables|/Script/Engine.Actor|GetRootComponent _asbuilding_actor)) "Movable") (Variables|/Script/FortniteGame.BuildingActor|GetbIsMovable _asbuilding_actor))
                            (LocalVariables|SetbRetBlock false)
                            (else
                              _)))
                        (:CastFailed
                          _))
                      (LocalVariables|SetbRetBlock true))))))))))
    (:then_1
      (return (LocalVariables|GetbRetBlock)))))

(fn InitData
  ;[0.01779 0.1562 0.1152] Cache out data to be used.
  (Networking|SwitchHasAuthority
    (:Authority
      (Utilities|FlowControl|Sequence
        (:then_0
          (bind _asbp_supply_drop_radio_flare_actor (Utilities|Casting|CastToBP_SupplyDropRadioFlareActor (Actor|GetOwner)))
          (bind _data (Class|BPSupplyDropRadioFlareActor|GetSupplyDropRadioData _asbp_supply_drop_radio_flare_actor))
          (Variables|Default|SetFallHeight (.FallHeight_47_E26A7B624EA7475B16BAE79339FD145D (Utilities|Struct|BreakSupplyDropRadioDataStruct _data)))
          (Variables|Default|SetGravityMultiplier_PreOpen (.GravityMultiplier_PreOpen_49_9A76CDC14234D269F36B459F31CC1011 (Utilities|Struct|BreakSupplyDropRadioDataStruct _data)))
          (Variables|Default|SetGravityMultiplier_PostOpen (.GravityMultiplier_PostOpen_51_BF7A9EE244D825B798B0E3B46418865C (Utilities|Struct|BreakSupplyDropRadioDataStruct _data)))
          (Variables|SupplyDropRadio|SetBalloonOpenedFallSpeed (.BalloonOpenedFallSpeed_53_A7105A3C4E29A83513426780577EF185 (Utilities|Struct|BreakSupplyDropRadioDataStruct _data)))
          (Variables|SupplyDropRadio|SetDistanceFromOriginalSpawnLocForBalloonOpen (.DistanceFromOriginalSpawnLocForBalloonOpen_56_B884693E44BCB97A947F1CB9ECADEBF4 (Utilities|Struct|BreakSupplyDropRadioDataStruct _data))))
        (:then_1
          (Variables|SupplyDropRadio|SetClientLocationCorrectionMaxDistSqWhenBalloonDisabled (ScalableFloat|GetValueatLevel (Variables|Hotfix|GetHotfix_ClientLocationCorrectionMaxDistSqWhenBalloonDisabled)))
          (Variables|Default|SetFlareDurationMultiplier (ScalableFloat|GetValueatLevel (Variables|Hotfix|GetHotfix_FlareDurationMultiplier)))
          (Variables|Default|SetMaxNumOfHeightTeleportAttempts (ScalableFloat|GetValueasInteger (Variables|Hotfix|GetHotfix_MaxNumOfHeightTeleportAttempts)))
          (Variables|Default|SetFallHeightTeleportFailIncrement (ScalableFloat|GetValueatLevel (Variables|Hotfix|GetHotfix_FallHeightTeleportFailIncrement)))
          (Variables|SupplyDropRadio|SetClientLocationCorrectionMaxDistSqWhenBalloonEnabled (ScalableFloat|GetValueatLevel (Variables|Hotfix|GetHotfix_ClientLocationCorrectionMaxDistSqWhenBalloonEnabled))))))))

(fn IsGrounded
  (return (Utilities|Struct|BreakSupplyDropRadioGroundInfoStruct (CallFunction|GetGroundInfo))))

(fn OnRep_bIsDying
  ;[0.05266 0.1821 0.1979] Disable collision once the actor is dying.
  (Collision|SetActorEnableCollision false))

(fn OnRep_GroundInfo
  (bind _groundinfo_bvalid_8_6eea3ba64f4126c2fde592aedf82163b (Utilities|Struct|BreakSupplyDropRadioGroundInfoStruct (Variables|GetGroundInfo_bValid_8_6EEA3BA64F4126C2FDE592AEDF82163B)))
  (bind (_groundinfo_bvalid_8_6eea3ba64f4126c2fde592aedf82163b_2 _groundinfo_groundactor_2_3cc9853b48ea31722d3aeba61f41499c _groundinfo_groundlocation_5_9924359a435909587f0606803cb2ef23) (Utilities|Struct|BreakSupplyDropRadioGroundInfoStruct (Variables|GetGroundInfo_bValid_8_6EEA3BA64F4126C2FDE592AEDF82163B)))
  (bind _groundinfo_bvalid_8_6eea3ba64f4126c2fde592aedf82163b_3 (Utilities|Struct|BreakSupplyDropRadioGroundInfoStruct (Variables|GetGroundInfo_bValid_8_6EEA3BA64F4126C2FDE592AEDF82163B)))
  (bind (_groundinfo_bvalid_8_6eea3ba64f4126c2fde592aedf82163b_4 _groundinfo_groundactor_2_3cc9853b48ea31722d3aeba61f41499c_2) (Utilities|Struct|BreakSupplyDropRadioGroundInfoStruct (Variables|GetGroundInfo_bValid_8_6EEA3BA64F4126C2FDE592AEDF82163B)))
  ;[0.05785 0.1726 0.2031] OnRep_GroundInfo: Called whenever the landed ground info changes.
  ; Runs on both server and client.
  ;;[0.05224 0.1562 0.1213] Log the landed/falling state of the actor.
  (Networking|SwitchHasAuthority
    (:Authority
      (Fortnite|FortShippingLog self (Utilities|String|Append "OnRep_GroundInfo: " (select _groundinfo_bvalid_8_6eea3ba64f4126c2fde592aedf82163b_4 (Utilities|String|Append " Landed on actor : " (Utilities|GetDisplayName _groundinfo_groundactor_2_3cc9853b48ea31722d3aeba61f41499c_2)) " Starting fall."))))
    (:Remote
      _))
  ;;[0.06139 0.2396 0.1296] Tick handles movement, so stop/start tick as necessary.
  (Actor|Tick|SetActorTickEnabled (not _groundinfo_bvalid_8_6eea3ba64f4126c2fde592aedf82163b))
  (Default|CallOnGroundInfoUpdate (Variables|Default|GetGroundInfo))
  (Networking|SwitchHasAuthority
    (:Authority
      (if _groundinfo_bvalid_8_6eea3ba64f4126c2fde592aedf82163b_3
        (Variables|SupplyDropRadio|SetbMoving false)))
    (:Remote
      (if _groundinfo_bvalid_8_6eea3ba64f4126c2fde592aedf82163b_2
        (Transformation|SetActorLocation _groundinfo_groundlocation_5_9924359a435909587f0606803cb2ef23) ; On the client, set the actor location to the ground location.
        (CallFunction|HandleLandEvent)))))

(fn SetupEntryFXTime
  (bind _entryfx (Variables|Default|GetEntryFX))
  ;[0.09375 0.02464 0.07911] SetupEntryFXTime: Setup for how long the entry FX should be played on this actor.
  (Networking|CanExecuteCosmeticEvents
    (:True
      (bind _value (CallFunction|GetClosedStateFallDuration))
      (if (> _value 0)
        (Effects|Components|ParticleSystem|SetFloatParameter _entryfx "Time" (+ _value (Variables|Default|GetEntryFXOffsetDuration)))
        (Components|Activation|Activate _entryfx false)))))

(fn ShouldAddAttachComponentForActor
  (params (InActor :type Actor%20Object%20Reference))
  ;[0.008247 0.07812 0.06084] Setup to be overridden in the child.
  (Utilities|FlowControl|Sequence
    (:then_0
      (Networking|SwitchHasAuthority
        (:Authority
          (return true))))
    (:then_1
      (return false))))

(fn SweepMovement_Deprecated
  (params (DeltaTime :type Float%20%28double-precision%29))
  (bind _returnvalue (Transformation|GetActorLocation))
  (bind _waterhitresult (Variables|SupplyDropRadio|GetWaterHitResult))
  ;[0.1562 0.05687 0.1536] SweepMovement: This function is responsible for sweep moving the actor towards the ground.
  ; It is run on both server and client.
  (Utilities|FlowControl|Sequence
    (:then_0
      ;;[0.0227 0.05183 0.1354] IF the actor has hit water, return the water hit result.
      ;; In the current design, the actor does not sink into water and thus no more movement is required.
      (if (Collision|BreakHitResult _waterhitresult)
        (return _waterhitresult)))
    (:then_1
      (Deprecated|UpdateBalloonFallSpeedDeprecated DeltaTime)
      ;;[0.06988 0.05627 0.1406] On the client, IF the replicated server location is over the error threshold when compared to the client actor location, the client actor is teleported to the server location.
      (Networking|CanExecuteCosmeticEvents
        (:True
          (LocalVariables|SetClientPreCorrectedLocation _returnvalue)
          (if (> (Math|Vector|DistanceSquared(Vector) _returnvalue (Variables|SupplyDropRadio|GetServerLocation)) (select (Utilities|Enum|Equal(Enum) (SupplyDropRadio|GetBalloonState) "Opened") (Variables|SupplyDropRadio|GetClientLocationCorrectionMaxDistSqWhenBalloonEnabled) (Variables|SupplyDropRadio|GetClientLocationCorrectionMaxDistSqWhenBalloonDisabled)))
            (Transformation|SetActorLocation (Variables|SupplyDropRadio|GetServerLocation))
            ;;[0.01949 0.07081 0.08333] Debug/Track the correction.
            (if (Variables|Default|GetbDebugMovement)
              (Rendering|Debug|DrawDebugBox (LocalVariables|GetClientPreCorrectedLocation) (Components|Box|GetScaledBoxExtent (Variables|Default|GetWorldCollision)) "(R=1.000000,G=0.000000,B=0.000000,A=1.000000)" (Transformation|GetWorldRotation (Variables|Default|GetWorldCollision)) 2.0 5.0)
              (Rendering|Debug|DrawDebugBox (Transformation|GetWorldLocation (Variables|Default|GetWorldCollision)) (Components|Box|GetScaledBoxExtent (Variables|Default|GetWorldCollision)) "(R=0.000000,G=1.000000,B=0.000000,A=1.000000)" (Transformation|GetWorldRotation (Variables|Default|GetWorldCollision)) 2.0 5.0)
              (Fortnite|FortShippingLog self "SweepMovement: ClientLocationCorrectionOccurred")
              (Development|PrintString "SweepMovement: ClientLocationCorrectionOccurred" true false "(R=0.593750,G=0.103442,B=0.079808,A=1.000000)" 5.0))))))
    (:then_2
      ;;[0.01448 0.1328 0.1458] Sweep move the actor based on the fallspeed.
      (bind _sweephitresult (Transformation|SetActorLocation (+ (Transformation|GetActorLocation) (Utilities|Operators|Multiply (Math|Vector|VectorDown) (Variables|SupplyDropRadio|GetFallSpeed) DeltaTime)) true))
      (return _sweephitresult))))

(fn TeleportToDesiredHeight
  ;[0 0.0625 0.03861] Teleport the actor to a valid location at desired height.
  (Networking|SwitchHasAuthority
    (:Authority
      (LocalVariables|SetTeleportDestination (Math|Vector|MakeVector (.x (Variables|SupplyDropRadio|GetOriginalServerSpawnLocation)) (.y (Variables|SupplyDropRadio|GetOriginalServerSpawnLocation)) (+ (.z (Variables|SupplyDropRadio|GetOriginalServerSpawnLocation)) (Variables|Default|GetFallHeight))))
      (LocalVariables|SetAttachedActors (Actor|GetAttachedActors true true))
      (for _index (range (+ (Math|Integer|Max(Integer) (- (Variables|Default|GetMaxNumOfHeightTeleportAttempts) 1)) 1))
        (bind (_outhit _returnvalue) (Collision|BoxTraceForObjects (+ (LocalVariables|GetTeleportDestination) (Variables|Default|GetCenterOfMassRelativeOffset)) (+ (+ (LocalVariables|GetTeleportDestination) (Variables|Default|GetCenterOfMassRelativeOffset)) "0,0,1.000000") (Variables|Default|GetSupplyDropActorHalfSize) "0, 0, 0" (LocalVariables|GetTraceObjectTypes) false (LocalVariables|GetAttachedActors)))
        (if (not _returnvalue)
          (bind _returnvalue_2 (Transformation|Teleport (LocalVariables|GetTeleportDestination) (Transformation|GetActorRotation)))
          (if _returnvalue_2
            (Fortnite|FortShippingLog self (Utilities|String|Append "TeleportToDesiredHeight: Success at  " (Utilities|String|ToString(Integer) (+ _index 1)) " attempts."))
            (return)
            (else
              _))
          (else
            _))
        (Variables|Default|SetFallHeight (+ (Variables|Default|GetFallHeight) (Variables|Default|GetFallHeightTeleportFailIncrement)))
        (LocalVariables|SetTeleportDestination (Math|Vector|MakeVector (.x (Variables|SupplyDropRadio|GetOriginalServerSpawnLocation)) (.y (Variables|SupplyDropRadio|GetOriginalServerSpawnLocation)) (+ (.z (Variables|SupplyDropRadio|GetOriginalServerSpawnLocation)) (Variables|Default|GetFallHeight)))))
      (Fortnite|FortShippingLog self "TeleportToDesiredHeight: Failed to teleport to desired height. Setting ActorLocation directly instead." true)
      (Transformation|SetActorLocation (LocalVariables|GetTeleportDestination)))))

(fn TickBalloonState_Deprecated
  (bind (_returnvalue_x _returnvalue_y _returnvalue_z) (Math|Vector|BreakVector (Transformation|GetActorLocation)))
  (bind (_originalserverspawnlocation_x _originalserverspawnlocation_y _originalserverspawnlocation_z) (Math|Vector|BreakVector (Variables|GetOriginalServerSpawnLocation_X)))
  ;[0.2604 0.08592 0.1198] TickBalloonState: If the balloon is closed, check if it is in a state where it should be opened.
  (Networking|SwitchHasAuthority
    (:Authority
      (if (Utilities|Enum|Equal(Enum) (SupplyDropRadio|GetBalloonState) "Closed")
        (if (or (< (- _returnvalue_z _originalserverspawnlocation_z) (Variables|SupplyDropRadio|GetDistanceFromOriginalSpawnLocForBalloonOpen)) (not (SupplyDropRadio|IsMoving)))
          (SupplyDropRadio|SetBalloonState "Opened"))))))

(fn TickMovement_Deprecated
  (params (DeltaTime :type Float%20%28double-precision%29))
  ;[0.0237 0.1198 0.1186] TickMovement: This function handles the movement of this actor towards the ground.
  ; It detects colliding actors using sweep and decides if those actors are valid actors to land on.
  ; If the colliding actor is NOT a valid actor, it is ignored and the actor continues moving towards the ground.
  ; If the colliding actor is valid, the ground info variable is updated and the actor's movement is stopped.
  ; The actor tries to move agian once the ground actor that it has landed on has been destroyed.
  (LocalVariables|SetPreviousLocation (Transformation|GetActorLocation))
  (bind _hitresult (Deprecated|SweepMovementDeprecated DeltaTime))
  (Variables|Default|SetLatestSweep _hitresult)
  (Deprecated|UpdateVelocityDeprecated DeltaTime (LocalVariables|GetPreviousLocation))
  (bind _bvalidblock (Deprecated|HandleMovementBlockedDeprecated _hitresult))
  (Networking|SwitchHasAuthority
    (:Authority
      (Variables|SupplyDropRadio|SetServerLocation (Transformation|GetActorLocation))
      (if _bvalidblock
        (bind _output_get (Variables|SetGroundInfo_bValid_8_6EEA3BA64F4126C2FDE592AEDF82163B (Utilities|Struct|MakeSupplyDropRadioGroundInfoStruct true (.HitActor (Collision|BreakHitResult _hitresult)) (.Location (Collision|BreakHitResult _hitresult)))))
        (bind (_output_get_bvalid_8_6eea3ba64f4126c2fde592aedf82163b _output_get_groundactor_2_3cc9853b48ea31722d3aeba61f41499c) (Utilities|Struct|BreakSupplyDropRadioGroundInfoStruct _output_get))
        (bind _asbuilding_actor (Utilities|Casting|CastToBuildingActor _output_get_groundactor_2_3cc9853b48ea31722d3aeba61f41499c)
          (:then
            (EventDispatchers|BindEventtoOnDied _asbuilding_actor (EventDispatchers|CreateEvent :!event HandleGroundBuildingActorDied)))
          (:CastFailed
            (Game|BindEventtoOnDestroyed (.HitActor (Collision|BreakHitResult _hitresult)) (EventDispatchers|CreateEvent :!event HandleGroundActorDestroyed))))
        (Utilities|FlowControl|Sequence
          (:then_0
            (bind _indicator (Variables|Default|GetIndicator)
              (:then
                (Class|SupplyDropRadioIndicator|SetOwnerLanded _indicator :!class /SupplyDropRadioGameplay/SupplyDrops/SupplyDropRadioIndicator.SupplyDropRadioIndicator_C))
              (:else)))
          (:then_1
            (CallFunction|HandleLandEvent)
            (if (Variables|SupplyDropRadio|GetbAutoOpen)
              (SupplyDropRadio|OpenSupplyDrop))))
        (elif (Utilities|Struct|BreakSupplyDropRadioGroundInfoStruct (Variables|Default|GetGroundInfo))
          (Variables|SetGroundInfo_bValid_8_6EEA3BA64F4126C2FDE592AEDF82163B (struct :bValid_8_6EEA3BA64F4126C2FDE592AEDF82163B false :GroundActor_2_3CC9853B48EA31722D3AEBA61F41499C 0 :GroundLocation_5_9924359A435909587F0606803CB2EF23 "0.000000,0.000000,0.000000")))))))

(fn UpdateBalloonFallSpeed_Deprecated
  (params (DeltaTime :type Float%20%28double-precision%29))
  ;[0.01776 0.1094 0.09296] UpdateBalloonFallSpeed: IF the actor is gounded, it falls at a constant unpopped speed.
  ; IF the actor is not grounded and has the balloon, it falls at a constant unpopped speed.
  ; IF the actor is not gounded and doesn't have the balloon, the fall speed increases cause of gravity acceleration.
  (LocalVariables|SetbGrounded (Utilities|Struct|BreakSupplyDropRadioGroundInfoStruct (Variables|Default|GetGroundInfo)))
  (switch Utilities|FlowControl|Switch|SwitchonESupplyDropRadioBalloonState (SupplyDropRadio|GetBalloonState)
    (:Closed
      _)
    (:Opened
      (Variables|SupplyDropRadio|SetFallSpeed (Variables|SupplyDropRadio|GetBalloonOpenedFallSpeed)))
    (:Destroyed
      _))
  (if (LocalVariables|GetbGrounded)
    (Variables|SupplyDropRadio|SetFallSpeed (Variables|SupplyDropRadio|GetBalloonClosedGroundedFallSpeed))
    (else
      (Variables|SupplyDropRadio|SetFallSpeed (+ (Variables|SupplyDropRadio|GetFallSpeed) (* DeltaTime (SupplyDropRadio|GetGravityZ)))))))

(fn UpdateGroundInfo
  (params (bValidBlock :type Boolean) (MovementHitResult :type Hit%20Result%20Structure%20%28by%20ref%29))
  (bind (_bblockinghit _binitialoverlap _time _distance _location _impactpoint _normal _impactnormal _physmat _hitactor) (Collision|BreakHitResult MovementHitResult))
  ;[0.04282 0.2969 0.2821] UpdateGroundInfo: Update the ground info and based on that value, take up necessary actions.
  (Networking|SwitchHasAuthority
    (:Authority
      (if bValidBlock
        (bind _output_get (Variables|SetGroundInfo_bValid_8_6EEA3BA64F4126C2FDE592AEDF82163B (Utilities|Struct|MakeSupplyDropRadioGroundInfoStruct true _hitactor _location)))
        (bind (_output_get_bvalid_8_6eea3ba64f4126c2fde592aedf82163b _output_get_groundactor_2_3cc9853b48ea31722d3aeba61f41499c) (Utilities|Struct|BreakSupplyDropRadioGroundInfoStruct _output_get))
        (bind _asbuilding_actor (Utilities|Casting|CastToBuildingActor _output_get_groundactor_2_3cc9853b48ea31722d3aeba61f41499c)
          (:then
            (EventDispatchers|BindEventtoOnDied _asbuilding_actor (EventDispatchers|CreateEvent :!event HandleGroundBuildingActorDied)))
          (:CastFailed
            (Game|BindEventtoOnDestroyed _hitactor (EventDispatchers|CreateEvent :!event HandleGroundActorDestroyed))))
        (Utilities|FlowControl|Sequence
          (:then_0
            (bind _indicator (Variables|Default|GetIndicator)
              (:then
                (Class|SupplyDropRadioIndicator|SetOwnerLanded _indicator :!class /SupplyDropRadioGameplay/SupplyDrops/SupplyDropRadioIndicator.SupplyDropRadioIndicator_C))
              (:else)))
          (:then_1
            (CallFunction|HandleLandEvent)
            (if (Variables|SupplyDropRadio|GetbAutoOpen)
              (SupplyDropRadio|OpenSupplyDrop))))
        (elif (Utilities|Struct|BreakSupplyDropRadioGroundInfoStruct (Variables|Default|GetGroundInfo))
          (Variables|SetGroundInfo_bValid_8_6EEA3BA64F4126C2FDE592AEDF82163B (struct :bValid_8_6EEA3BA64F4126C2FDE592AEDF82163B false :GroundActor_2_3CC9853B48EA31722D3AEBA61F41499C 0 :GroundLocation_5_9924359A435909587F0606803CB2EF23 "0.000000,0.000000,0.000000")))))))

(fn UpdateVelocity_Deprecated
  (params (DeltaTime :type Float%20%28double-precision%29) (PreviousLocation :type Vector))
  (bind (_bblockinghit _binitialoverlap _time _distance _location _impactpoint _normal _impactnormal _physmat _hitactor) (Collision|BreakHitResult (Variables|Default|GetLatestSweep)))
  ;[0.02017 0.1599 0.1615] Update the velocity of the actor.
  (Networking|SwitchHasAuthority
    (:Authority
      (if (> DeltaTime 0)
        (bind _returnvalue (SupplyDropRadio|GetHighestAttachParent _hitactor))
        (if (not (== _returnvalue self))
          (bind _output_get (Variables|SupplyDropRadio|SetVelocity (/ (- (Transformation|GetActorLocation) PreviousLocation) DeltaTime)))
          (Variables|SupplyDropRadio|SetbMoving (> (Math|Vector|VectorLengthSquared _output_get) (Variables|SupplyDropRadio|GetMinimumMovingSpeedThresholdSq))))))))

(fn ConstructionScript)
