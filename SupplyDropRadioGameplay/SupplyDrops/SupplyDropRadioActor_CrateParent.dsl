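; Base supply drop radio actor for crate-type drops, wrapping a loot crate held in a child actor component.
; On the server it auto-marks on spawn, spawns crate loot when opened, and applies a quest-search GameplayEffect to the crate on landing.
; Remote clients play landing sound and FX at landing, choosing water or land FX from the ground actor.
; Marking targets the child crate, falling back to the parent result.
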(event EventBeginPlay
  ;[0.04573 0.2292 0.0608] BeginPlay: Auto mark the crate actor, if desired.
  (|CallParentFunction)
  (Networking|SwitchHasAuthority
    (:Authority
      (bind _crateactorchild (Variables|Default|GetCrateActorChild)
        (:then
          (CallFunction|AutoMarkonSpawn (Variables|/Script/Engine.ChildActorComponent|GetChildActor _crateactorchild)))
        (:else)))))

(event SupplyDropRadio|EventOpenSupplyDrop
  ;[0.03042 0.1719 0.1702] Open the supply drop crate.
  (|CallParentFunction)
  (Networking|SwitchHasAuthority
    (:Authority
      (bind _crateactorchild (Variables|Default|GetCrateActorChild)
        (:then
          (bind _assupply_drop_radio_item_crate_parent (Utilities|Casting|CastToSupplyDropRadioItem_CrateParent (Variables|/Script/Engine.ChildActorComponent|GetChildActor _crateactorchild)))
          (Class|SupplyDropRadioItemCrateParent|SpawnSupplyDropLoot _assupply_drop_radio_item_crate_parent 0))
        (:else)))))

(fn GetAttachedActorToMark
  ;[0.2292 0.03675 0.2053] Return the crate child actor.
  (bind _actor (|CallParentFunction))
  (Utilities|FlowControl|Sequence
    (:then_0
      (bind _crateactorchild (Variables|Default|GetCrateActorChild)
        (:then
          (return (Variables|/Script/Engine.ChildActorComponent|GetChildActor _crateactorchild)))
        (:else)))
    (:then_1
      (return _actor))))

(fn HandleLandEvent
  (bind (_groundinfo_bvalid_8_6eea3ba64f4126c2fde592aedf82163b _groundinfo_groundactor_2_3cc9853b48ea31722d3aeba61f41499c _groundinfo_groundlocation_5_9924359a435909587f0606803cb2ef23) (Utilities|Struct|BreakSupplyDropRadioGroundInfoStruct (CallFunction|GetGroundInfo)))
  (bind _crateactorchild (Variables|Default|GetCrateActorChild))
  (bind _childactor (Variables|/Script/Engine.ChildActorComponent|GetChildActor _crateactorchild))
  (bind _returnvalue (Ability|GetAbilitySystemComponent _childactor))
  (bind (_asfort_water_body_actor _bsuccess) (Utilities|Casting|CastToFortWaterBodyActor _groundinfo_groundactor_2_3cc9853b48ea31722d3aeba61f41499c))
  ;[0.01586 0.09038 0.1146] HandleLandEvent: Called when the supply drop has landed.
  ; On the server, apply quest search supply drop GE.
  ; On the client, play the land effects.
  (|CallParentFunction)
  (Networking|SwitchHasAuthority
    (:Authority
      (Utilities|IsValid _childactor
        (:Is%20Valid
          (Utilities|IsValid _returnvalue
            (:Is%20Valid
              (GameplayEffects|ApplyGameplayEffectToSelf :self _returnvalue :GameplayEffectClass (Variables|Default|GetQuestSearchSupplyDropGE) :Level 0.0))))))
    (:Remote
      (Audio|PlaySoundatLocation (Variables|Default|GetLandingSoundCue) _groundinfo_groundlocation_5_9924359a435909587f0606803cb2ef23)
      (Effects|Components|ParticleSystem|SpawnFXAtLocation (select _bsuccess (Variables|Default|GetLandingFXSystem_Water) (Variables|Default|GetLandingFXSystem_Land)) (Math|Vector|MakeVector (.x (.location (Transformation|GetWorldTransform (Variables|Default|GetWorldCollision)))) (.y (.location (Transformation|GetWorldTransform (Variables|Default|GetWorldCollision)))) (- (.z (.location (Transformation|GetWorldTransform (Variables|Default|GetWorldCollision)))) (.z (Variables|/Script/Engine.BoxComponent|GetBoxExtent (Variables|Default|GetWorldCollision)))))))))

(fn ConstructionScript
  (|CallParentFunction))
