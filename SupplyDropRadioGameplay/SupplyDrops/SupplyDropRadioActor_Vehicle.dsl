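; Supply drop radio actor that delivers a vehicle, showing a child-actor preview in the editor that is destroyed on BeginPlay.
; The server spawns the vehicle, fills its fuel and applies mod tags, while OnRep_SpawnedVehicle attaches it to the drop.
; Opening the drop destroys the balloon, detaches the vehicle and breaks its sleep lock.
; Zero vehicle health also triggers opening and sets destruction info, as does the deprecated movement-blocked handler, without the destruction step.
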(event EventBeginPlay
  ;[0.03398 0.07787 0.1146] BeginPlay: Destroy the preview mesh, in PIE.
  ; Spawn the vehicle desired vehicle actor.
  (|CallParentFunction)
  (Utilities|FlowControl|Sequence
    (:then_0
      (bind _vehiclepreview_editoronly (Variables|Default|GetVehiclePreview_EDITORONLY)
        (:then
          (Components|DestroyComponent _vehiclepreview_editoronly))
        (:else)))
    (:then_1
      (Networking|SwitchHasAuthority
        (:Authority
          (bind _returnvalue (Game|SpawnActorfromClass (Variables|Default|GetVehicleClass) (Transformation|GetWorldTransform (Variables|Default|GetVehicleAttachComponent)) "AlwaysSpawn" "SelectDefaultAtRuntime" 0 true 0))
          (bind _output_get (Variables|Default|SetSpawnedVehicle _returnvalue)) ; The OnRep attaches it to this actor 
          (Utilities|IsValid _output_get
            (:Is%20Valid
              (Vehicle|Fuel|SetFuel _returnvalue (Vehicle|Fuel|GetFuelCapacity _returnvalue))
              (Health|AssignOnHealthChanged _returnvalue (EventDispatchers|CreateEvent :!event Custom|HandleOnHealthChanged))
              (CallFunction|AutoMarkonSpawn _returnvalue)
              (for _array_element (GameplayTags|BreakGameplayTagContainer (Variables|Default|GetVehicleModTags))
                (Vehicle|Mod|ApplyVehicleMod _returnvalue _array_element 0)))
            (:Is%20Not%20Valid
              ;;[1 0 0] Warning: Failed to spawn a vehicle.
              (Fortnite|FortShippingLog self "BeginPlay: Failed to spawn a vehicle." true))))))))

(event SupplyDropRadio|EventOpenSupplyDrop
  ;[0.08871 0.2693 0.2865] OpenSupplyDrop: Destroy the balloon actor and wake up the vehicle.
  (|CallParentFunction)
  (Networking|SwitchHasAuthority
    (:Authority
      (CallFunction|DestroyBalloonActor false)
      (bind _spawnedvehicle (Variables|Default|GetSpawnedVehicle)
        (:then
          (Transformation|DetachFromActor _spawnedvehicle "KeepWorld" "KeepWorld" "KeepWorld")
          (Vehicle|BreakSpawnedSleepLock _spawnedvehicle)
          (Health|UnbindEventfromOnHealthChanged _spawnedvehicle (EventDispatchers|CreateEvent :!event Custom|HandleOnHealthChanged))
          (Fortnite|FortShippingLog self (Utilities|String|Append "OpenSupplyDrop: Breaking lock for vehicle: " (Utilities|GetDisplayName _spawnedvehicle))))
        (:else
          ;;[1 0 0] Warning: Spawned vehicle is null.
          (Fortnite|FortShippingLog self "OpenSupplyDrop: SpawnedVehicle is null." true))))))

(event Custom|HandleOnHealthChanged
  ;[0.1875 0.04257 0.1698] HandleOnHealthChanged: Callback for whenever the vehicle's health changes.
  (Networking|SwitchHasAuthority
    (:Authority
      (bind _spawnedvehicle (Variables|Default|GetSpawnedVehicle)
        (:then
          (if (<= (Class|FortVehicleInterface|GetHealth _spawnedvehicle) 0)
            (SupplyDropRadio|OpenSupplyDrop)
            (bind _asfort_dagwood_vehicle (Utilities|Casting|CastToFortDagwoodVehicle _spawnedvehicle))
            (Class|FortAthenaVehicle|ServerSetDestructionInfo _asfort_dagwood_vehicle (Class|FortAthenaVehicle|GetSelfDestructDelay _asfort_dagwood_vehicle) (Class|FortAthenaVehicle|GetMaxZeroHealthTime _asfort_dagwood_vehicle) (Class|FortAthenaVehicle|GetSelfDestructCleanupDelay _asfort_dagwood_vehicle) 0.0)))
        (:else)))))

(event EventEndPlay
  (params (EndPlayReason :type EEndPlayReason%20Enum :default "Destroyed"))
  ;[0.1302 0.07275 0.02709] EndPlay: Unbind delegates
  (|CallParentFunction EndPlayReason)
  (Networking|SwitchHasAuthority
    (:Authority
      (bind _spawnedvehicle (Variables|Default|GetSpawnedVehicle)
        (:then
          (Health|UnbindEventfromOnHealthChanged _spawnedvehicle (EventDispatchers|CreateEvent :!event Custom|HandleOnHealthChanged)))
        (:else)))))

(fn GetAttachedActorToMark
  ;[0.1719 0.01489 0.1326] Return the vehicle child actor.
  (bind _actor (|CallParentFunction))
  (Utilities|FlowControl|Sequence
    (:then_0
      (return (Variables|Default|GetSpawnedVehicle)))
    (:then_1
      (return _actor))))

(fn HandleMovementBlocked_Deprecated
  (params (HitResult :type Hit%20Result%20Structure%20%28by%20ref%29))
  (bind (_bblockinghit _binitialoverlap _time _distance _location _impactpoint _normal _impactnormal _physmat _hitactor) (Collision|BreakHitResult HitResult))
  ;[0.03304 0.1991 0.2188] HandleMovementBlocked: If the movement ever gets blocked cause of hitting a pawn (that is not the vehicle itself), the vehicle is released.
  (bind _bvalidblock (|CallParentFunction HitResult))
  (Utilities|FlowControl|Sequence
    (:then_0
      (Networking|SwitchHasAuthority
        (:Authority
          (bind _asfort_pawn (Utilities|Casting|CastToFortPawn _hitactor))
          (if (!= (Variables|Default|GetSpawnedVehicle) _asfort_pawn)
            (Fortnite|FortShippingLog self (Utilities|String|Append "HandleMovementBlocked: Releaseing vehicle cause of movement being blocked by pawn: " (Fortnite|GetPlayerorObjectNameSafe _asfort_pawn)))
            (SupplyDropRadio|OpenSupplyDrop)))))
    (:then_1
      (return _bvalidblock))))

(fn OnRep_SpawnedVehicle
  ;[0.02007 0.1998 0.25] OnRep_SpawnedVehicle: Attach the spawned vehicle to the vehicle attach component on both the server and client.
  (bind _spawnedvehicle (Variables|Default|GetSpawnedVehicle)
    (:then
      (Transformation|AttachActorToComponent _spawnedvehicle (Variables|Default|GetVehicleAttachComponent) "None" "SnapToTarget" "KeepRelative" "KeepRelative" true)
      (Collision|IgnoreActorwhenMoving (Class|FortVehicleInterface|GetSkeletalMeshComponent _spawnedvehicle) self true)
      (Fortnite|FortShippingLog self (Utilities|String|Append "OnRep_SpawnedVehicle: Spawned vehicle: " (Utilities|GetDisplayName _spawnedvehicle))))
    (:else)))

(fn ShouldAddAttachComponentForActor
  (params (InActor :type Actor%20Object%20Reference))
  (bind (_asbuilding_gameplay_actor _bsuccess) (Utilities|Casting|CastToBuildingGameplayActor InActor))
  ;[0.02109 0.07502 0.08333] Do not change the relevancy range for the cowcatcher.
  (Utilities|FlowControl|Sequence
    (:then_0
      (Utilities|IsValid InActor
        (:Is%20Valid
          (bind _returnvalue (GameplayTags|HasMatchingGameplayTag(Message) InActor "(TagName=\"Vehicle.Mod.CowCatcher\")"))
          (if (and _returnvalue _bsuccess)
            (return false)))))
    (:then_1
      (bind _badd (|CallParentFunction InActor))
      (return _badd))))

(fn ConstructionScript
  (|CallParentFunction)
  (ChildActorComponent|SetChildActorClass (Variables|Default|GetVehiclePreview_EDITORONLY) (Variables|Default|GetVehicleClass)))
