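; Parent balloon for radio supply drops, a damageable building actor that notifies listeners when shot down.
; Server damage lazily spawns a timed health bar, and death fires analytics and a destroyed dispatcher, with separate pop effects.
; The replicated parent drop drives collision, visibility, light and audio through balloon-state and ground-info delegates.
; While falling, a timer plays audio only with listeners in range, saving movement cost via manual attachment.
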(event Building|EventOnBuildingActorInitialized
  (params (InitializationReason :type EFortBuildingInitializationReason%20Enum :default "StaticallyPlaced") (BuildingPersistentState :type EFortBuildingPersistentState%20Enum :default "Default"))
  ;[0 0.08333 0.008499] BuildingActorInitialized: Add gameplay effect to reduce incoming damage.
  (|CallParentFunction InitializationReason BuildingPersistentState)
  (Fortnite|FortShippingLog self "BuildingActorInitialized: Invoked.")
  (Environment|ApplyGlobalEnvironmentGameplayEffecttoActor self (Variables|Default|GetGE_SupplyDropBalloonDamageReduction)))

(event EventOnDamageServer
  (params (Damage :type Float%20%28single-precision%29) (DamageTags :type Gameplay%20Tag%20Container%20Structure%20%28by%20ref%29) (Momentum :type Vector :default "0, 0, 0") (HitInfo :type Hit%20Result%20Structure%20%28by%20ref%29) (InstigatedBy :type Controller%20Object%20Reference) (DamageCauser :type Actor%20Object%20Reference) (EffectContext :type Gameplay%20Effect%20Context%20Handle%20Structure))
  ;[0.09437 0.02095 0.1146] OnDamageServer: When the actor is damaged, spawn a health bar component since we want to start showing the health of the actor.
  (Networking|SwitchHasAuthority
    (:Authority
      (Variables|Default|GetHealthBarComponent
        (:then
          (HealthBarComponent|SetHealthBarVisibility (Variables|Default|GetHealthBarComponent) true))
        (:else
          (bind _returnvalue (Game|AddComponentbyClass 0 "/SupplyDropRadioGameplay/SupplyDrops/SupplyDropRadioBalloonHealthBarComponent.SupplyDropRadioBalloonHealthBarComponent_C" true (Variables|Default|GetHealthBarComponentRelativeTransform)))
          (bind _output_get (Variables|Default|SetHealthBarComponent _returnvalue))
          (Utilities|IsValid _output_get
            (:Is%20Valid
              (HealthBarComponent|SetHealthBarVisibility _output_get true)
              (Fortnite|FortShippingLog self "OnDamageServer: Health bar component spawned."))
            (:Is%20Not%20Valid
              ;;[1 0 0] Health bar failed to spawn.
              (Fortnite|FortShippingLog self "OnDamageServer: Health bar failed to spawn." true)))))
      (Utilities|Time|ClearandInvalidateTimerbyHandle (Variables|Default|GetHealthbar%20Visible%20Timer%20Handle))
      (bind _returnvalue_2 (Utilities|Time|SetTimerbyEvent (delegate Custom|OnHealthbarVisibleTimerExpired) (Variables|Default|GetHealthbarVisibleTime)))
      (Variables|Default|SetHealthbar%20Visible%20Timer%20Handle _returnvalue_2))))

(event Custom|OnHealthbarVisibleTimerExpired
  ;[0.09437 0.02095 0.1146] OnDamageServer: When the actor is damaged, spawn a health bar component since we want to start showing the health of the actor.
  (HealthBarComponent|SetHealthBarVisibility (Variables|Default|GetHealthBarComponent) false))

(event EventOnDeathServer
  (params (Damage :type Float%20%28single-precision%29) (DamageTags :type Gameplay%20Tag%20Container%20Structure%20%28by%20ref%29) (Momentum :type Vector :default "0, 0, 0") (HitInfo :type Hit%20Result%20Structure%20%28by%20ref%29) (InstigatedBy :type Controller%20Object%20Reference) (DamageCauser :type Actor%20Object%20Reference) (EffectContext :type Gameplay%20Effect%20Context%20Handle%20Structure))
  ;[0.01449 0.09896 0.06553] OnDeathServer: When the balloon is destroyed, fire delegate, analytics and hide the health bar visibility.
  (Networking|SwitchHasAuthority
    (:Authority
      (Analytics|FireAnalyticsEventBlueprint (Variables|Default|GetSupplyDropShotDownAnalyticsEvent) (Utilities|Struct|MakeFortAnalyticsEventAttribute "(Name=\"\",Value=\"\")"))
      (Default|CallOnBalloonDestroyed) ; Let supply drop know
      (Collision|SetActorEnableCollision false)
      (Fortnite|FortShippingLog self "OnDeathServer: Invoked.")
      (CallFunction|RecordBalloonDeathAnalytics InstigatedBy)
      (bind _healthbarcomponent (Variables|Default|GetHealthBarComponent)
        (:then
          (HealthBarComponent|SetHealthBarVisibility _healthbarcomponent false))
        (:else)))))

(event EventOnDeathPlayEffects
  (params (Damage :type Float%20%28single-precision%29) (DamageTags :type Gameplay%20Tag%20Container%20Structure%20%28by%20ref%29) (Momentum :type Vector :default "0, 0, 0") (HitInfo :type Hit%20Result%20Structure%20%28by%20ref%29) (InstigatedBy :type Fort%20Pawn%20Object%20Reference) (DamageCauser :type Actor%20Object%20Reference) (EffectContext :type Gameplay%20Effect%20Context%20Handle%20Structure))
  (bind _balloonpopfxmeshsocket (Variables|Default|GetBalloonPopFXMeshSocket))
  (bind _staticmeshcomponent (Variables|Debug|GetStaticMeshComponent))
  (bind _returnvalue (Transformation|GetSocketLocation _staticmeshcomponent _balloonpopfxmeshsocket))
  ;[0.151 0.03702 0.1119] OnDeathPlayEffects: Hide the mesh, disable the collision, play the balloon destroy FX, stop the falling audio.
  (Fortnite|FortShippingLog self "OnDeathPlayEffects: Invoked.")
  (Rendering|SetActorHiddenInGame true)
  (Collision|SetActorEnableCollision false)
  (Niagara|SpawnSystematLocation (Variables|Default|GetBalloonPopFX_NS) _returnvalue)
  (Audio|PlaySoundatLocation (Variables|Default|GetBalloonPopSound) _returnvalue)
  (Audio|Components|Audio|Stop (Variables|Default|GetFallingAudio))
  (Utilities|Time|ClearandInvalidateTimerbyHandle (Variables|Default|GetHandleFallingAudioStateTimerHandle))
  (bind _supplydropradioactorparent (Variables|Default|GetSupplyDropRadioActorParent)
    (:then
      (Default|UnbindEventfromOnGroundInfoUpdate _supplydropradioactorparent (EventDispatchers|CreateEvent :!event HandleGroundInfoUpdated)))
    (:else)))

(event EventEndPlay
  (params (EndPlayReason :type EEndPlayReason%20Enum :default "Destroyed"))
  ;[0.1094 0.06823 0.02242] EndPlay: Unbind delegates.
  (|CallParentFunction EndPlayReason)
  (Fortnite|FortShippingLog self "EndPlay: Invoked.")
  (bind _supplydropradioactorparent (Variables|Default|GetSupplyDropRadioActorParent)
    (:then
      (Default|UnbindEventfromOnGroundInfoUpdate _supplydropradioactorparent (EventDispatchers|CreateEvent :!event HandleGroundInfoUpdated)))
    (:else)))

(fn GetAttachedActorToMark
  ;[0.7912 0.3422 1] Return the attached actor from the supply drop radio actor parent.
  (Utilities|FlowControl|Sequence
    (:then_0
      (bind _supplydropradioactorparent (Variables|Default|GetSupplyDropRadioActorParent)
        (:then
          (bind _actor (Class|SupplyDropRadioActorParent|GetAttachedActortoMark _supplydropradioactorparent))
          (return _actor))
        (:else)))
    (:then_1
      (return 0))))

(fn HandleBalloonStateChanged
  (params (BalloonState :type ESupplyDropRadioBalloonState%20Enum :default "Invalid"))
  (bind _staticmeshcomponent (Variables|Debug|GetStaticMeshComponent))
  (bind _ns_supplydrop_light (Variables|Default|GetNS_SupplyDrop_Light))
  (LocalVariables|SetbOpened (Utilities|Enum|Equal(Enum) BalloonState "Opened"))
  (Utilities|FlowControl|Sequence
    (:then_0
      (if (LocalVariables|GetbOpened)
        (Collision|SetCollisionProfileName _staticmeshcomponent "FortBuildingMeshPhysics" false :!class /Script/Engine.PrimitiveComponent)
        (else
          (Collision|SetCollisionResponsetoAllChannels _staticmeshcomponent "ECR_Ignore")
          (for _array_element (Utilities|Array|MakeArray "ECC_WorldStatic" "ECC_WorldDynamic" "ECC_Pawn" "ECC_GameTraceChannel4")
            (Collision|SetCollisionResponsetoChannel _staticmeshcomponent _array_element "ECR_Block")))))
    (:then_1
      (if (LocalVariables|GetbOpened)
        (Audio|Components|Audio|FadeIn (Variables|Default|GetFallingAudio) 0.2 1.0 0.0 "Linear" :!class /Script/Engine.AudioComponent)
        (Rendering|SetVisibility (Variables|Debug|GetStaticMeshComponent) true false)
        (Components|Activation|Activate _ns_supplydrop_light false)
        (else
          (Audio|Components|Audio|FadeOut (Variables|Default|GetFallingAudio) 0.2 0.0 "Linear" :!class /Script/Engine.AudioComponent)
          (Rendering|SetVisibility (Variables|Debug|GetStaticMeshComponent) false false)
          (Components|Activation|Deactivate _ns_supplydrop_light))))))

(fn HandleGroundInfoUpdated
  (params (GroundInfo_bValid_8_6EEA3BA64F4126C2FDE592AEDF82163B :type Boolean) (GroundInfo_GroundActor_2_3CC9853B48EA31722D3AEBA61F41499C :type Actor%20Object%20Reference) (GroundInfo_GroundLocation_5_9924359A435909587F0606803CB2EF23 :type Vector))
  (bind _fallingaudio (Variables|Default|GetFallingAudio))
  ;[0 0.1771 0.1291] Delegate callback whenever the parent supply drop lands.
  ; Updates the state of the falling audio.
  (Networking|CanExecuteCosmeticEvents
    (:True
      (Fortnite|FortShippingLog self (Utilities|String|Append "HandleGroundInfoUpdated: GroundInfoValid : " (Utilities|String|ToString(Boolean) GroundInfo_bValid_8_6EEA3BA64F4126C2FDE592AEDF82163B) " : GroundActor: " (Utilities|GetDisplayName GroundInfo_GroundActor_2_3CC9853B48EA31722D3AEBA61F41499C)))
      (if GroundInfo_bValid_8_6EEA3BA64F4126C2FDE592AEDF82163B
        (Utilities|Time|ClearandInvalidateTimerbyHandle (Variables|Default|GetHandleFallingAudioStateTimerHandle)) ; If landed, stop falling audio.
        (if (Audio|Components|Audio|IsPlaying _fallingaudio)
          (Audio|Components|Audio|FadeOut _fallingaudio 0.25 0.0 "Linear" :!class /Script/Engine.AudioComponent))
        (else
          (bind _returnvalue (Utilities|Time|SetTimerbyEvent (EventDispatchers|CreateEvent :!event UpdateFallingAudioState) (Variables|Default|GetUpdateFallingAudioStateInterval) true)) ; If falling, start falling audio.
          (Variables|Default|SetHandleFallingAudioStateTimerHandle _returnvalue)
          (CallFunction|UpdateFallingAudioState))))))

(fn OnRep_SupplyDropRadioActorParent
  ;[0 0.1198 0.0848] Bind to receive a callback whenever the supply drop lands.
  (bind _supplydropradioactorparent (Variables|Default|GetSupplyDropRadioActorParent)
    (:then
      (Default|BindEventtoOnGroundInfoUpdate _supplydropradioactorparent (EventDispatchers|CreateEvent :!event HandleGroundInfoUpdated))
      (CallFunction|HandleGroundInfoUpdated (Class|SupplyDropRadioActorParent|GetGroundInfo _supplydropradioactorparent))
      (Fortnite|FortShippingLog self (Utilities|String|Append "OnRep_SupplyDropRadioActorParent: Set as " (Utilities|GetDisplayName _supplydropradioactorparent)))
      (Default|BindEventtoOnBalloonStateChanged _supplydropradioactorparent (EventDispatchers|CreateEvent :!event HandleBalloonStateChanged))
      (CallFunction|HandleBalloonStateChanged (SupplyDropRadio|GetBalloonState _supplydropradioactorparent)))
    (:else)))

(fn PlaySpawnMaterialAnimation
  (bind _staticmeshcomponent (Variables|Debug|GetStaticMeshComponent))
  ;[0.02709 0.2277 0.2396] Play Animation
  (Networking|CanExecuteCosmeticEvents
    (:True
      (Rendering|SetVisibility _staticmeshcomponent true false)
      (Rendering|Material|SetCustomPrimitiveDataFloat _staticmeshcomponent 0 (Utilities|Time|GetTimeSeconds)))))

(fn RecordBalloonDeathAnalytics
  (params (Controller :type Controller%20Object%20Reference))
  (bind _returnvalue (Fortnite|Athena|GameState|GetGameStateAthena))
  (bind _returnvalue_2 (Actor|GetComponentbyClass _returnvalue "/SupplyDropRadioGameplay/Gameplay/BP_SupplyDropRadioGameStateComponent.BP_SupplyDropRadioGameStateComponent_C"))
  ;[0.04691 0.1861 0.1979] Record the death of the balloon analytics if it was instigated by a player while the balloon was falling.
  (Networking|SwitchHasAuthority
    (:Authority
      (Utilities|Casting|CastToFortPlayerController Controller)
      (bind _supplydropradioactorparent (Variables|Default|GetSupplyDropRadioActorParent)
        (:then
          (if (not (Utilities|Struct|BreakSupplyDropRadioGroundInfoStruct (Class|SupplyDropRadioActorParent|GetGroundInfo _supplydropradioactorparent)))
            (Utilities|IsValid _returnvalue
              (:Is%20Valid
                (Utilities|IsValid _returnvalue_2
                  (:Is%20Valid
                    (Analytics|AnalyticsIncrementPlayerDestroyedFallingBalloon _returnvalue_2)))))))
        (:else)))))

(fn SetSupplyDropRadioActorParent
  (params (InParentActor :type Supply%20Drop%20Radio%20Actor%20Parent%20Object%20Reference))
  (Networking|SwitchHasAuthority
    (:Authority
      (Variables|Default|SetSupplyDropRadioActorParent InParentActor))))

(fn UpdateFallingAudioState
  ;[0 0.1719 0.1484] UpdateFallingAudioState: Since the falling audio component uses manual attachment, it only remains attached if the component is actually active. This helps with movement perf.
  ; Therefore, when there are NO listeners within the range of the actor, stop the audio component.
  (Networking|CanExecuteCosmeticEvents
    (:True
      (bind _fallingaudio (Variables|Default|GetFallingAudio)
        (:then
          (bind _sound (Variables|/Script/Engine.AudioComponent|GetSound _fallingaudio)
            (:then
              (bind _returnvalue (Audio|AreAnyListenersWithinRange (Transformation|GetActorLocation) (Variables|/Script/Engine.SoundBase|GetMaxDistance _sound)))
              (if _returnvalue
                (if (not (Audio|Components|Audio|IsPlaying _fallingaudio))
                  (Audio|Components|Audio|FadeIn _fallingaudio 1.0 1.0 0.0 "Linear" :!class /Script/Engine.AudioComponent))
                (else
                  (if (Audio|Components|Audio|IsPlaying _fallingaudio)
                    (Audio|Components|Audio|FadeOut _fallingaudio 0.2 0.0 "Linear" :!class /Script/Engine.AudioComponent)))))
            (:else)))
        (:else)))))

(fn ConstructionScript)
