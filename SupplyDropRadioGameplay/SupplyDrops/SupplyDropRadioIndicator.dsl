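; Supply drop radio ground indicator that marks where a drop will land with flare and reticle FX, audio and a sound indicator.
; The server runs timed downward traces to set the replicated GroundLocation, which moves the actor via OnRep.
; It binds to the hit actor's death or destruction, then ignores that actor and traces again.
; Replicated booleans fade out flare and reticle after a timer or landing, which also sets a short lifespan.
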(event EventBeginPlay
  (bind _groundlocation (Variables|Default|GetGroundLocation))
  ;[0.007855 0.1927 0.02135] BeginPlay: On the server, start a timer to figure out what the ground location is.
  ; On the server, start a timer at the end of which the flare FX should be turned off.
  ; On the client, update the actor location to the replicated gorund location and fade in the audio components.
  (|CallParentFunction)
  (Fortnite|FortShippingLog self "BeginPlay: Invoked.")
  (Networking|SwitchHasAuthority
    (:Authority
      (Variables|Default|SetGroundLocation (Transformation|GetActorLocation))
      (CallFunction|StartUpdateGroundLocationChecks 1.5)
      (Utilities|Time|SetTimerbyEvent (EventDispatchers|CreateEvent :!event Custom|DisableFlare) (Variables|Default|GetFlareDuration)))
    (:Remote
      (Audio|Components|Audio|FadeIn (Variables|Default|GetFlareAudio) 3 1.0 0.0 "Linear" :!class /Script/Engine.AudioComponent)
      (Transformation|SetActorLocation _groundlocation)
      (Audio|PlaySoundatLocation (Variables|Default|GetMarkerAudio) _groundlocation)
      (Fortnite|ShowSoundIndicatorLocalforTeam :Location _groundlocation :MaxAudibleDistance (Variables|/Script/Engine.SoundBase|GetMaxDistance (Variables|/Script/Engine.AudioComponent|GetSound (Variables|Default|GetFlareAudio))) :IndicatorType "Gunshot" :Team (Team|GetActorTeam self) :Affiliations (Utilities|Array|MakeArray "Hostile") :Tint "(R=0.000000,G=0.000000,B=0.000000,A=0.000000)" :AttenuationMaxRangeMultiplier 1.0 :AttenuationIntensityMultiplier 1.0))))

(event EventEndPlay
  (params (EndPlayReason :type EEndPlayReason%20Enum :default "Destroyed"))
  ;[0.1562 0.08007 0.01206] EndPlay: Unbind delegates.
  (|CallParentFunction EndPlayReason)
  (Fortnite|FortShippingLog self "EndPlay: Invoked.")
  (Networking|SwitchHasAuthority
    (:Authority
      (CallFunction|UnbindGroundDestroyedDelegates))))

(event Custom|DisableFlare
  ;[0.007715 0.07292 0.06418] DisableFlare: Replicate boolean to let the clients know that the time to disable the flare has come!
  (Networking|SwitchHasAuthority
    (:Authority
      (Variables|Default|SetbDisableFlare true))))

(fn HandleGroundActorDestroyed
  (params (DestroyedActor :type Actor%20Object%20Reference))
  ;[0.01754 0.1146 0.1066] HandleGroundActorDestroyed: Delegate callback when the ground actor of the indicator is destroyed.
  ; When this happens, start searching for a new ground actor.
  (Networking|SwitchHasAuthority
    (:Authority
      (Utilities|IsValid DestroyedActor
        (:Is%20Valid
          (Game|UnbindEventfromOnDestroyed DestroyedActor (EventDispatchers|CreateEvent :!event HandleGroundActorDestroyed))
          (Utilities|Array|Add (Variables|Default|GetPersistentTraceIgnoreActors) DestroyedActor)
          (Fortnite|FortShippingLog self (Utilities|String|Append "HandleGroundActorDestroyed: GroundActor destroyed. " "CachedGroundActor = " (Utilities|GetDisplayName (Variables|Default|GetGroundActor)) " ; FunctionInputGroundActor = " (Utilities|GetDisplayName DestroyedActor)) (!= (Variables|Default|GetGroundActor) (Variables|GetDestroyedActor)))
          (Utilities|Array|RemoveItem (Variables|Default|GetPersistentTraceIgnoreActors)) ; Remove any nulls.
          (CallFunction|StartUpdateGroundLocationChecks 0.0)
          (CallFunction|UpdateGroundLocation))))))

(fn HandleGroundBuildingActorDestroyed
  (params (DamagedActor :type Actor%20Object%20Reference) (Damage :type Float%20%28single-precision%29) (InstigatedBy :type Controller%20Object%20Reference) (DamageCauser :type Actor%20Object%20Reference) (HitLocation :type Vector) (FHitComponent :type Primitive%20Component%20Object%20Reference) (BoneName :type Name) (Momentum :type Vector))
  ;[0.01811 0.0969 0.1042] HandleGroundBuildingActorDestroyed: Delegate callback for when the ground building actor is destroyed.
  (Networking|SwitchHasAuthority
    (:Authority
      (bind _asbuilding_actor (Utilities|Casting|CastToBuildingActor DamagedActor))
      (EventDispatchers|UnbindEventfromOnDied _asbuilding_actor (EventDispatchers|CreateEvent :!event HandleGroundBuildingActorDestroyed))
      (CallFunction|HandleGroundActorDestroyed _asbuilding_actor))))

(fn OnRep_bDisableFlare
  ;[0.01057 0.1146 0.1089] Disable the flare FX.
  (Fortnite|FortShippingLog self "OnRep_bDisableFlare: Invoked.")
  (Components|Activation|Deactivate (Variables|Default|GetFlare_Niagara))
  (Audio|Components|Audio|FadeOut (Variables|Default|GetFlareAudio) 3.0 0.0 "Linear" :!class /Script/Engine.AudioComponent))

(fn OnRep_bOwnerLanded
  ;[0.002728 0.1094 0.0773] OnRep_bOwnerLanded: Called when the supply drop has landed.
  ; Disable the flare and deactivate the reticle.
  (Fortnite|FortShippingLog self "OnRep_bOwnerLanded: Invoked.")
  (Networking|SwitchHasAuthority
    (:Authority
      (CallFunction|DisableFlare))
    (:Remote
      (Components|Activation|Deactivate (Variables|Default|GetReticle_Niagara))
      (Audio|Components|Audio|FadeOut (Variables|Default|GetReticleAudio) 0.2 0.0 "Linear" :!class /Script/Engine.AudioComponent))))

(fn OnRep_GroundLocation
  ;[0.03615 0.1433 0.1771] Set the actor to the latest ground location.
  ; Runs on both server and client.
  (Transformation|SetActorLocation (Variables|Default|GetGroundLocation)))

(fn SetOwnerLanded
  ;[0.01003 0.1198 0.1042] Called from the supply drop when it has landed.
  (Networking|SwitchHasAuthority
    (:Authority
      (Variables|Default|SetbOwnerLanded true)
      (Actor|SetLifeSpan 5.0))))

(fn StartUpdateGroundLocationChecks
  (params (InitialStartDelay :type Float%20%28single-precision%29))
  ;[0 0.1127 0.125] StartUpdateGroundLocationChecks: Start a timer that finds the ground.
  (Networking|SwitchHasAuthority
    (:Authority
      (Fortnite|FortShippingLog self "StartUpdateGroundLocationChecks: Invoked")
      (bind _returnvalue (Utilities|Time|SetTimerbyEvent (EventDispatchers|CreateEvent :!event UpdateGroundLocation) (Variables|Default|GetUpgradeGroundLocation) true false InitialStartDelay))
      (Variables|Default|SetUpgradeGroundLocationTimerHandle _returnvalue))))

(fn UnbindGroundDestroyedDelegates
  ;[0.005066 0.1266 0.1406] Unbind delegates related to the current ground actor's destruction.
  (Networking|SwitchHasAuthority
    (:Authority
      (bind _groundactor (Variables|Default|GetGroundActor)
        (:then
          (bind _asbuilding_actor (Utilities|Casting|CastToBuildingActor _groundactor)
            (:then
              (EventDispatchers|UnbindEventfromOnDied _asbuilding_actor (EventDispatchers|CreateEvent :!event HandleGroundBuildingActorDestroyed)))
            (:CastFailed
              (Game|UnbindEventfromOnDestroyed _groundactor (EventDispatchers|CreateEvent :!event HandleGroundActorDestroyed)))))
        (:else)))))

(fn UpdateGroundLocation
  (bind _returnvalue (Actor|GetOwner))
  ;[0.03025 0.1719 0.165] UpdateGroundLocation: On the server, run a trace from the supply drop (owner) to the ground to figure out where this actor should be moved to.
  (Networking|SwitchHasAuthority
    (:Authority
      (Utilities|IsValid _returnvalue
        (:Is%20Valid
          (bind _output_get (LocalVariables|SetIgnoreActors (Actor|GetAttachedActors _returnvalue true false)))
          (Utilities|Array|Add _output_get _returnvalue)
          (Utilities|Array|AppendArray _output_get (Variables|Default|GetPersistentTraceIgnoreActors))
          (bind (_outhit _returnvalue_2) (Collision|LineTraceByProfile (Variables|Default|GetGroundLocation) (+ (Variables|Default|GetGroundLocation) (* (Math|Vector|VectorDown) (Variables|Default|GetFallHeight))) "FortSupplyDropGroundTrace" false (LocalVariables|GetIgnoreActors)))
          (if _returnvalue_2
            (Utilities|IsValid (.HitActor (Collision|BreakHitResult _outhit))
              (:Is%20Valid
                (Fortnite|FortShippingLog self (Utilities|String|Append "UpdateGroundLocation: HitActor: " (Utilities|GetDisplayName (.HitActor (Collision|BreakHitResult _outhit)))))
                (Variables|Default|SetGroundLocation (.ImpactPoint (Collision|BreakHitResult _outhit))) ; Note: This sets the location of the \nactor through the OnRep \ncall on both server and client.
                (CallFunction|UnbindGroundDestroyedDelegates)
                (Utilities|Time|ClearandInvalidateTimerbyHandle (Variables|Default|GetUpgradeGroundLocationTimerHandle))
                (bind _output_get_2 (Variables|Default|SetGroundActor (.HitActor (Collision|BreakHitResult _outhit))))
                (bind _asbuilding_actor (Utilities|Casting|CastToBuildingActor _output_get_2)
                  (:then
                    (EventDispatchers|BindEventtoOnDied _asbuilding_actor (EventDispatchers|CreateEvent :!event HandleGroundBuildingActorDestroyed)))
                  (:CastFailed
                    (Game|BindEventtoOnDestroyed _output_get_2 (EventDispatchers|CreateEvent :!event HandleGroundActorDestroyed)))))
              (:Is%20Not%20Valid
                (Fortnite|FortShippingLog self "UpdateGroundLocation: Trace is blocking hit but HitActor is null." true)))))))))

(fn ConstructionScript)
