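; Parent Blueprint for the radio supply drop crate, attached to a supply drop actor that gates interaction by movement and ground state.
; Interaction or death triggers server-side loot once, guarded by bOpened, spawning pickups one per frame, then destroying the balloon and notifying the parent actor.
; Interaction also applies quest gameplay effects.
; OnRep_bOpened plays looted cosmetics, while death plays destruction effects if the crate is unopened.
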(event Interaction|EventBlueprintOnInteract
  (params (InteractingPawn :type Fort%20Pawn%20Object%20Reference) (InteractionBeingAttempted :type EInteractionBeingAttempted%20Enum :default "FirstInteraction"))
  ;[0.0765 0.1452 0.1823] BlueprintOnInteract: Called when a player interacts with the crate.
  (Networking|SwitchHasAuthority
    (:Authority
      (Fortnite|FortShippingLog self (Utilities|String|Append "BlueprintOnInteract: Invoked by player :" (Fortnite|GetPlayerNameSafe InteractingPawn)))
      (CallFunction|SpawnSupplyDropLoot InteractingPawn)
      (CallFunction|HandleInteractQuests InteractingPawn))))

(event Custom|SpawnLoot
  (params (ItemsToSpawn :type Array%20of%20Fort%20Item%20Entry%20Structures) (TriggeringPawn :type Fort%20Pawn%20Object%20Reference))
  ;[0.02723 0.09896 0.09007] SpawnLoot: Spawn loot one frame at a time for the optimal collision interaction between the various spawned drops  (implemented similarly to the AthenaSupplyDrop)
  (Networking|SwitchHasAuthority
    (:Authority
      (bind _outelementarray (Default|GetSubsetOfArraywithDelayLoop ItemsToSpawn 1)
        (:Loop
          (for _array_element _outelementarray
            (CallFunction|SpawnPickup (Utilities|Casting|CastToFortWorldItemDefinition (Class|FortItemFunctionLibrary|BreakFortItemEntry _array_element)) (.Count (Class|FortItemFunctionLibrary|BreakFortItemEntry _array_element)) TriggeringPawn (Transformation|GetActorLocation))))) ; Runs once per frame.
    )))

(event EventOnDeathServer
  (params (Damage :type Float%20%28single-precision%29) (DamageTags :type Gameplay%20Tag%20Container%20Structure%20%28by%20ref%29) (Momentum :type Vector :default "0, 0, 0") (HitInfo :type Hit%20Result%20Structure%20%28by%20ref%29) (InstigatedBy :type Controller%20Object%20Reference) (DamageCauser :type Actor%20Object%20Reference) (EffectContext :type Gameplay%20Effect%20Context%20Handle%20Structure))
  ;[0.1875 0.126 0.1715] OnDeathServer: Disable collision, set life span on the actor, and spawn the supply drop loot if it hasn't been spawned yet.
  (Collision|SetActorEnableCollision false)
  (Actor|SetLifeSpan 2.0)
  (CallFunction|SpawnSupplyDropLoot 0)
  (Fortnite|FortShippingLog self "OnDeathServer: Invoked"))

(event EventOnDeathPlayEffects
  (params (Damage :type Float%20%28single-precision%29) (DamageTags :type Gameplay%20Tag%20Container%20Structure%20%28by%20ref%29) (Momentum :type Vector :default "0, 0, 0") (HitInfo :type Hit%20Result%20Structure%20%28by%20ref%29) (InstigatedBy :type Fort%20Pawn%20Object%20Reference) (DamageCauser :type Actor%20Object%20Reference) (EffectContext :type Gameplay%20Effect%20Context%20Handle%20Structure))
  (bind _cratemesh (Variables|Default|GetCrateMesh))
  (bind _returnvalue (Transformation|GetWorldTransform _cratemesh))
  (bind _location (.location _returnvalue))
  ;[0.07421 0.1889 0.1927] OnDeathPlayEffects - When the supply drop is destroyed, rather than looted.
  (if (not (Variables|Default|GetbOpened))
    (Niagara|SpawnSystematLocation "/Game/Effects/Fort_Effects/World/Destruction/Base_Effects/NS_General_Deco_Destruction_Small_Converted.NS_General_Deco_Destruction_Small_Converted" _location (.rotation _returnvalue))
    (Audio|PlaySoundatLocation (Variables|Default|GetSoundDestroyed) _location)
    (Rendering|SetActorHiddenInGame true)
    (Fortnite|FortShippingLog self "OnDeathPlayEffects: Invoked.")))

(fn BlueprintCanInteract
  (params (InteractingPawn :type Fort%20Pawn%20Object%20Reference) (InteractionBeingAttempted :type EInteractionBeingAttempted%20Enum :default "FirstInteraction") (InteractionType :type TInteractionType%20Enum :default "IT_NoInteraction"))
  (bind _actorparent (CallFunction|GetSupplyDropRadioActorParent))
  ;[0.1458 0.03005 0.1035] The player can interact with the supply drop as long as the
  (Utilities|FlowControl|Sequence
    (:then_0
      (Utilities|IsValid _actorparent
        (:Is%20Valid
          (return (or (not (SupplyDropRadio|IsMoving _actorparent)) (Utilities|Struct|BreakSupplyDropRadioGroundInfoStruct (Class|SupplyDropRadioActorParent|GetGroundInfo _actorparent)))))))
    (:then_1
      (return false))))

(fn BlueprintGetInteractionString
  (params (InteractingPawn :type Fort%20Pawn%20Object%20Reference) (InteractionBeingAttempted :type EInteractionBeingAttempted%20Enum :default "FirstInteraction"))
  (return (Variables|GetInteractionText)))

(fn GetSupplyDropRadioActorParent
  (return (Utilities|Casting|CastToSupplyDropRadioActorParent (Actor|GetAttachParentActor))))

(fn HandleInteractQuests
  (params (InteractingPlayerPawn :type Fort%20Pawn%20Object%20Reference))
  (bind _returnvalue (Ability|GetAbilitySystemComponent InteractingPlayerPawn))
  ;[0.01644 0.04661 0.1094] Apply interact quest GEs for the interacting player.
  (Networking|SwitchHasAuthority
    (:Authority
      (GameplayEffects|ApplyGameplayEffectToSelf :self _returnvalue :GameplayEffectClass "/Game/Athena/Items/QuestInteractables/Quest_GEs/GE_Athena_Quests_PostSupplyDropInteract.GE_Athena_Quests_PostSupplyDropInteract_C" :Level 0.0)
      (GameplayEffects|ApplyGameplayEffectToTarget :self _returnvalue :GameplayEffectClass "/Game/Athena/Items/Quests/Feats/GEs/GE_Feat_S11_RecentlyOpenedSupplyDrop.GE_Feat_S11_RecentlyOpenedSupplyDrop_C" :Target _returnvalue :Level 0.0)
      (Fortnite|FortShippingLog self (Utilities|String|Append "HandleInteractQuest: Applying interact quest GEs for player : " (Fortnite|GetPlayerNameSafe InteractingPlayerPawn))))))

(fn OnRep_bOpened
  (bind _cratemesh (Variables|Default|GetCrateMesh))
  (bind _returnvalue (Transformation|GetWorldTransform _cratemesh))
  (bind _location (.location _returnvalue))
  ;[0.03029 0.1458 0.1415] OnRep_bOpened: Supply drop has been opened, hide and disable collision.
  ; Play opened visuals on the client.
  (if (Variables|Default|GetbOpened)
    (Fortnite|FortShippingLog self "OnRep_bOpened: Invoked.")
    (Collision|SetActorEnableCollision false)
    (Networking|CanExecuteCosmeticEvents
      (:True
        (Development|SetHiddeninGame _cratemesh true true)
        (Effects|Components|ParticleSystem|SpawnFXAtLocation (Variables|Default|GetLootedFX) _location (.rotation _returnvalue))
        (Audio|SpawnSoundatLocation (Variables|Default|GetSoundLooted) _location)))))

(fn SpawnSupplyDropLoot
  (params (FortPawn :type Fort%20Pawn%20Object%20Reference))
  (bind _actorparent (CallFunction|GetSupplyDropRadioActorParent))
  ;[0.02016 0.1405 0.1615] SpawnSupplyDropLoot: Spawn the supply drop loot from the interaction of a player.
  (Networking|SwitchHasAuthority
    (:Authority
      (Fortnite|FortShippingLog self (Utilities|String|Append "SpawnSupplyDropLoot: Invoked by player : " (Fortnite|GetPlayerNameSafe FortPawn)))
      (if (not (Variables|Default|GetbOpened))
        (Variables|Default|SetbOpened true)
        (Networking|ForceNetUpdate)
        (Actor|SetLifeSpan 5.0)
        (Fortnite|FortShippingLog self (Utilities|String|Append "SpawnSupplyDropLoot: LootTableName: " (Utilities|String|ToString(Name) (Variables|FortAthenaSupplyDrop|GetLootTableNameOverride))))
        (bind (_outloottodrop _returnvalue) (Loot|PickLootDrops (Variables|FortAthenaSupplyDrop|GetLootTableNameOverride) (Loot|GetLootLevel) -1 "(GameplayTags=)"))
        (if _returnvalue
          (Analytics|SendSpawnLootAnalytics (Fortnite|GetFortPlayerControllerfromActor FortPawn) _outloottodrop)
          (CallFunction|SpawnLoot _outloottodrop FortPawn)
          (CallFunction|OnLootSpawned)
          (else
            ;;[1 0 0] Warning: Failed to pick loot drops
            (Fortnite|FortShippingLog self (Utilities|String|Append "SpawnSupplyDropLoot: Failed to spawn LootTableName: " (Utilities|String|ToString(Name) (Variables|FortAthenaSupplyDrop|GetLootTableNameOverride))) true)))
        (Utilities|IsValid _actorparent
          (:Is%20Valid
            (Class|SupplyDropRadioActorParent|DestroyBalloonActor _actorparent true)
            (Class|SupplyDropRadioActorParent|NotifySupplyDropComplete _actorparent)))))))

(fn ConstructionScript)
